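# Editor utility widget for world builders working with the PCG mesh grass system.
# Buttons toggle standalone and landscape grass visibility, set detail mode and PCG editor frame time, force partitioner or grass regeneration, and open a heatmap via Python.
# UpdatePerfMetrics finds grass actors in partitions around the viewport camera by querying actor descriptors.
# It reports component and instance counts by quality, plus selection point totals, flagging missing data or incomplete neighborhoods.
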EditorUtilityWidgetBlueprint /Game/Athena/Helios/Environments/Landscape/WorldBuildingTools/PCGGrass/Utilities/EUW_MeshGrass
graphs [
 EvaluateTargetActors {nodes [
   n1 = EvaluateTargetActors!() -> (
    then
    TargetActor:Array of Actor Object References
    MeshCountMedium:"Integer (by ref)"
    MeshCountEpic:"Integer (by ref)"
    InstanceCountMedium:"Integer (by ref)"
    InstanceCountEpic:"Integer (by ref)"
   ) {is=title K2Node_FunctionEntry}
   > Branch(Condition=Is Not Empty!(TargetArray:Array of Actor Object References=@n1) -> (ReturnValue:bool))
   > n2 = For Each Loop(Array:Array of Actor Object References=@n1) -> (Array Element:Actor Object Reference)
   > n3 = For Each Loop(Array:Array of Instanced Static Mesh Component Object References=Get Components By Class!(self:Actor Object Reference=@n2, ComponentClass:Actor Component Class Reference=/Script/Engine.InstancedStaticMeshComponent) -> (ReturnValue:Array of Instanced Static Mesh Component Object References)) -> (Array Element:Instanced Static Mesh Component Object Reference)
   > n4 = Branch(Condition="Equal (Enum)"(
     A:EDetailMode Enum=Get DetailMode!(self:Scene Component Object Reference=@n3) -> (DetailMode:EDetailMode Enum=DM_Low)
     B:EDetailMode Enum=DM_Epic
    ))
   n5 = GetInstanceCount!(self:Instanced Static Mesh Component Object Reference=@n3) -> (ReturnValue:int=0)
   > "Set Integer (by ref)"!(execute:Exec, Target:"Integer (by ref)"=@n1.InstanceCountEpic, Value:int=int + int(A=@n5, B=@n1.InstanceCountEpic)) -> (then:Exec) {is=title K2Node_VariableSetRef}
   > Increment Int(Value=@n1.MeshCountEpic)
   "Set Integer (by ref)"(execute=@n4.else, Target=@n1.InstanceCountMedium, Value=int + int(A=@n5, B=@n1.InstanceCountMedium))
   > Increment Int(Value=@n1.MeshCountMedium)
  ]}
 EventGraph {nodes [
   n1 = GetActorOfClass!(
    execute
    WorldContextObject:Object Reference
    ActorClass:Actor Class Reference=/Game/Athena/Helios/Environments/Landscape/WorldBuildingTools/PCGGrass/BP_MeshGrass_Partitioner.BP_MeshGrass_Partitioner_C
   ) -> (then, ReturnValue:BP Mesh Grass Partitioner Object Reference)
   > n2 = GetActorOfClass(ActorClass=/Script/PCG.PCGPartitionActor) -> (ReturnValue:PCGPartition Actor Object Reference)
   n3 = GetComponentByClass!(self:Actor Object Reference=@n2, ComponentClass:Actor Component Class Reference=/Script/PCG.PCGComponent) -> (ReturnValue:PCGComponent Object Reference)
   n4 = Get GraphInstance!(self:PCGComponent Object Reference=@n3) -> (GraphInstance:PCGGraph Instance Object Reference)
   > Branch(Condition="Equal (Object)"(
     A=GetConstPCGGraph!(self:PCGGraph Interface Object Reference=@n4) -> (ReturnValue:PCGGraph Object Reference)
     B=/Game/Athena/Helios/Environments/Landscape/WorldBuildingTools/PCGGrass/PCG_MeshGrass_Partitioner
    ))
   n5 = Break PCGData Collection!(PCGDataCollection:"PCGData Collection Structure (by ref)"=GetGeneratedGraphOutput!(self:PCGComponent Object Reference=@n3) -> (ReturnValue:"PCGData Collection Structure (by ref)")) -> (TaggedData:Array of PCGTagged Data Structures, bCancelExecutionOnEmpty:bool=~) {is=title K2Node_BreakStruct}
   > n6 = GetStringAttributeByMetadataKey!(
    execute
    Key:Integer64=0
    Metadata:PCGMetadata Object Reference=ConstMetadata!(self:PCGData Object Reference="Get (a copy)"!(Array:Array of PCGTagged Data Structures=@n5, Dimension 1) -> (Output_Data:PCGData Object Reference, Output_Tags:Set of Strings, Output_Pin:Name=None, Output_bIsUsedMultipleTimes:bool)) -> (ReturnValue:PCGMetadata Object Reference)
    AttributeName:Name=Date
   ) -> (then, ReturnValue:String)
   > LogString(InString=@n6)
   "On Released (BT_Repopulate)"() {is=title K2Node_ComponentBoundEvent}
   > n7 = GetAllActorsOfClass!(
    execute
    WorldContextObject:Object Reference
    ActorClass:Actor Class Reference=&BP_MeshGrass_Partitioner_C
   ) -> (then, OutActors:Array of BP Mesh Grass Partitioner Object References)
   > n8 = For Each Loop(Array:Array of BP Mesh Grass Partitioner Object References=@n7) -> (Array Element:BP Mesh Grass Partitioner Object Reference)
   > GenerateLocal!(
    execute
    self:PCGComponent Object Reference=Get PCG_MeshGrass_Partitioner!(self:BP Mesh Grass Partitioner Object Reference=@n8) -> (PCG_MeshGrass_Partitioner:PCGComponent Object Reference)
    bForce:bool=true
   ) -> (then)
   "On Released (EditorUtilityButton_23)"() {is=title K2Node_ComponentBoundEvent}
   > n9 = Flip Flop!(None:Exec) -> (A:Exec, B:Exec, IsA:bool=~) {is=title K2Node_MacroInstance}
   n10 = GetAllActorsOfClass(execute=[@n9.A, @n9.B], ActorClass=/Game/Athena/Helios/Environments/Landscape/WorldBuildingTools/PCGGrass/BP_MeshGrass.BP_MeshGrass_C) -> (OutActors:Array of BP Mesh Grass Object References)
   > n11 = For Each Loop(Array:Array of BP Mesh Grass Object References=@n10) -> (Array Element:BP Mesh Grass Object Reference)
   n12 = Get DefaultSceneRoot!(self:BP Mesh Grass Object Reference=@n11) -> (DefaultSceneRoot:Scene Component Object Reference)
   > Branch(Condition=IsVisible(self:Scene Component Object Reference=@n12))
   > SetVisibility!(
    execute
    self:Scene Component Object Reference=GetChildrenComponents!(self:Scene Component Object Reference=@n12, bIncludeAllDescendants:bool=false) -> (Children:Array of Scene Component Object References)
    bNewVisibility:bool=NOT Boolean(A=@n9)
    bPropagateToChildren:bool=false
   ) -> (then)
   "On Released (BT_Regenerate)"() {is=title K2Node_ComponentBoundEvent}
   > n13 = GetAllActorsWithTag!(execute, WorldContextObject:Object Reference, Tag:Name=PCG_MeshGrass) -> (then, OutActors:Array of Actor Object References)
   > n14 = For Each Loop(Array:Array of Actor Object References=@n13) -> (Array Element:Actor Object Reference)
   > GenerateLocal(self=GetComponentByClass(self=@n14, ComponentClass=&PCGComponent), bForce=true)
   "On Released (EditorUtilityButton)"() {is=title K2Node_ComponentBoundEvent}
   > n15 = Flip Flop() -> (IsA=~)
   > ExecuteConsoleCommand(Command=grass.Enable 0)
   > n16 = ExecuteConsoleCommand(Command=grass.FlushCache)
   n17 = GetAllActorsOfClass(execute=[@n16.then, @n20.then], ActorClass=&PCGPartitionActor) -> (OutActors:Array of PCGPartition Actor Object References)
   > n18 = For Each Loop(Array:Array of PCGPartition Actor Object References=@n17) -> (Array Element:PCGPartition Actor Object Reference)
   n19 = Get Components By Class(self=@n18, ComponentClass=/Script/PCG.PCGProceduralISMComponent) -> (ReturnValue:Array of PCGProcedural ISMComponent Object References)
   > Branch(Condition=Is Not Empty(TargetArray:Array of PCGProcedural ISMComponent Object References=@n19))
   > SetVisibility(self=@n19, bNewVisibility=@n15, bPropagateToChildren=false)
   n20 = ExecuteConsoleCommand(execute=@n15.B, Command=grass.Enable 1)
   "On Released (BT_Repopulate_1)"() {is=title K2Node_ComponentBoundEvent}
   > UpdatePerfMetrics()
   "On Released (BT_Heatmap)"() {is=title K2Node_ComponentBoundEvent}
   > OpenHeatmap()
   "On Released (EditorUtilityButton_4)"() {is=title K2Node_ComponentBoundEvent}
   > n21 = Flip Flop() -> (IsA=~)
   > n22 = ExecuteConsoleCommand(Command=grass.enable 1)
   SetBackgroundColor!(
    execute=[@n28.then, @n22.then]
    self:Button Object Reference=Get EditorUtilityButton_4() -> (EditorUtilityButton_4:Editor Utility Button Object Reference)
    InBackgroundColor:Linear Color Structure=SelectColor(A=(R=0.6,G=0.6,B=0.6,A=1), B=(R=1,G=1,B=1,A=1), bPickA=@n21)
   ) -> (then)
   > n23 = GetAllActorsOfClass(ActorClass=/Game/Athena/Helios/Environments/Landscape/WorldBuildingTools/PCGGrass/BP_MeshGrass_LandscapeGenerator_GPU.BP_MeshGrass_LandscapeGenerator_GPU_C) -> (OutActors:Array of BP Mesh Grass Landscape Generator GPU Object References)
   > n24 = For Each Loop(Array:Array of BP Mesh Grass Landscape Generator GPU Object References=@n23) -> (Array Element:BP Mesh Grass Landscape Generator GPU Object Reference)
   n25 = Get PCG!(self:BP Mesh Grass Landscape Generator GPU Object Reference=@n24) -> (PCG:PCGComponent Object Reference)
   > n26 = Set bActivated!(execute, bActivated:bool=@n21, self:PCGComponent Object Reference=@n25) -> (then, Output_Get:bool)
   > n27 = Branch(Condition=@n26)
   Refresh PCG Runtime Component!(execute=@n27.else, InComponent:PCGComponent Object Reference=@n25, bFlushCache:bool=false) -> (then)
   ExecuteConsoleCommand(execute=@n21.B, Command=grass.enable 0)
   > n28 = ExecuteConsoleCommand(Command=grass.FlushCache)
   GetLocalComponent!(self:PCGPartition Actor Object Reference=@n2, OriginalComponent:PCGComponent Object Reference=Get PCG_MeshGrass_Partitioner(self=@n1)) -> (ReturnValue:PCGComponent Object Reference)
   K2Node_Composite_0 = Speed Controls!() {is=title K2Node_Composite}
   K2Node_Composite_1 = Quality Controls!() {is=title K2Node_Composite}
  ]}
 Speed Controls {
  path=EventGraph.K2Node_Composite_0.Speed Controls
  nodes [
   "On Released (BT_Speed_6x)"() {is=title K2Node_ComponentBoundEvent}
   > ExecuteConsoleCommand(Command=pcg.EditorFrameTime 60)
   "On Released (BT_Speed_1x)"() {is=title K2Node_ComponentBoundEvent}
   > ExecuteConsoleCommand(Command=pcg.EditorFrameTime 8)
   "On Released (BT_Speed_3x)"() {is=title K2Node_ComponentBoundEvent}
   > ExecuteConsoleCommand(Command=pcg.EditorFrameTime 30)
   Inputs!()
   Outputs()
  ]
 }
 Quality Controls {
  path=EventGraph.K2Node_Composite_1.Quality Controls
  nodes [
   "On Released (EditorUtilityButton_1)"() {is=title K2Node_ComponentBoundEvent}
   > ExecuteConsoleCommand(Command=r.DetailMode 0)
   "On Released (EditorUtilityButton_3)"() {is=title K2Node_ComponentBoundEvent}
   > ExecuteConsoleCommand(Command=r.DetailMode 1)
   "On Released (EditorUtilityButton_2)"() {is=title K2Node_ComponentBoundEvent}
   > ExecuteConsoleCommand(Command=r.DetailMode 3)
   Inputs()
   Outputs()
  ]
 }
 FindMeshGrassActorByPartitionBounds {nodes [
   FindMeshGrassActorByPartitionBounds!() -> (then, Debug:bool=~, GridExtent:double=~, GridMargin:double=~, TargetBox:Box Structure, TargetClass:String) {is=title K2Node_FunctionEntry}
   n1 = Break Box!(Box:"Box Structure (by ref)"=Get TargetBox() -> (TargetBox:Box Structure)) -> (Min:Vector=~, Max:Vector=~, IsValid:bool=~) {is=title K2Node_BreakStruct}
   > n2 = Branch(Condition=@n1.IsValid)
   n3 = vector + vector(A=vector / vector(A:Vector=vector - vector(A=@n1.Max, B=@n1), B=2) -> (ReturnValue:Vector), B=@n1)
   n4 = float * float(A=Get GridExtent() -> (GridExtent:double), B=float - float(A=1, B=Get GridMargin() -> (GridMargin:double))) {is=title K2Node_PromotableOperator}
   > n5 = GetIntersectingActorDescs!(execute, InBox_Min:Vector=vector - vector(A=@n3, B:double=@n4), InBox_Max:Vector=vector + vector(A=@n3, B:double=@n4)) -> (then, OutActorDescs:Array of Actor Desc Structures, ReturnValue:bool)
   > n6 = Branch(Condition=@n5.ReturnValue)
   > n7 = For Each Loop(Array:Array of Actor Desc Structures=@n5) -> (Array Element:Actor Desc Structure)
   n8 = Break Actor Desc!(ActorDesc:"Actor Desc Structure (by ref)"=@n7) -> (
    Class_PathString:String
    Name:Name
    Label:Name
    Bounds:Box Structure
    RuntimeGrid:Name
    bIsSpatiallyLoaded:bool=~
    bActorIsEditorOnly:bool=~
    ActorPackage:Name
    ActorPath:Name
    FolderPath:Name
    DataLayerAssets:Array of Soft Object Path Structures
   ) {is=title K2Node_BreakStruct}
   > Branch(Condition="Equal Exactly (String)"(A=@n8, B=Get TargetClass() -> (TargetClass:String)))
   > n9 = GetAllActorsOfClass(ActorClass=&BP_MeshGrass_C) -> (OutActors:Array of BP Mesh Grass Object References)
   > n10 = Branch(Condition=Is Not Empty(TargetArray:Array of BP Mesh Grass Object References=@n9))
   > n11 = Filter by ID Name!(
    execute
    TargetArray:Array of Object References=@n9
    NameSubString:String="To String (Name)"(InName=@n8.Name)
    StringMatch:EEditorScriptingStringMatchType Enum=ExactMatch
    FilterType:EEditorScriptingFilterType Enum
   ) -> (then, ReturnValue:Array of BP Mesh Grass Object References)
   > n12 = Branch(Condition=Is Not Empty(TargetArray=@n11))
   > n13 = Branch(Condition=Get Debug() -> (Debug:bool))
   n14 = "Get (a copy)"!(Array:Array of BP Mesh Grass Object References=@n11, Dimension 1) -> (Output:BP Mesh Grass Object Reference)
   > n15 = LogString(InString=Append(A="Selected Actor: ", B=GetDisplayName(Object=@n14)))
   Return Node!(execute=[@n15.then, @n13.else], TargetActor:Actor Object Reference=@n14, Success:bool=true)
   Return Node(execute=[@n2.else, @n6.else, @n7.Completed, @n10.else, @n12.else], Success=false)
  ]}
 OpenHeatmap {nodes [
   OpenHeatmap() {is=title K2Node_FunctionEntry}
   > ExecutePythonCommand(PythonCommand=Replace(SourceString=Get Command() -> (Command:String), From="##", To="\r\n"))
  ]}
 ParseInt {nodes [
   n1 = ParseInt!() -> (then, Int:int) {is=title K2Node_FunctionEntry}
   n2 = "To String (Integer)"(InInt=@n1)
   n3 = Len(S=@n2)
   > Return Node!(
    execute
    ReturnValue:String=Append(
     A=@n2
     B=Select!(Option 0:String, Option 1:String="   ", Option 2:String="  ", Option 3:String=" ", Option 4:String, Option 5:String, Index:int=@n3) -> (ReturnValue:String)
    )
   )
  ]}
 UpdatePerfMetrics {nodes [
   UpdatePerfMetrics() {is=title K2Node_FunctionEntry}
   > n1 = Sequence!(execute) -> (then_0, then_1, then_2:Exec)
   > n2 = Sequence!(execute) -> (then_0, then_1)
   > n3 = Get Editor Camera Info!(execute) -> (then, CameraLocation_X:double, CameraLocation_Y:double, CameraLocation_Z:double, ForwardDirection:Vector, FOV:float, ReturnValue:bool) {is=title K2Node_CallFunction}
   n4 = MakeVector(X=@n3, Y=@n3.CameraLocation_Y)
   n5 = float - float(A=1, B=Get GridMargin())
   n6 = Get GridExtent()
   n7 = vector / vector(A:double=float * float(A=float * float(A=@n6, B=@n5), B=4), B=2)
   > n8 = GetIntersectingActorDescs(InBox_Min=vector - vector(A=@n4, B:Vector=@n7), InBox_Max=vector + vector(A=@n4, B:Vector=@n7))
   > Branch(Condition=@n8.ReturnValue)
   > n9 = For Each Loop(Array=@n8)
   n10 = Break Actor Desc!(ActorDesc=@n9) -> (
    NativeClass:Object Class Reference
    Label
    Bounds
    RuntimeGrid
    bIsSpatiallyLoaded=~
    bActorIsEditorOnly=~
    ActorPackage
    ActorPath
    FolderPath
    DataLayerAssets
   )
   > Branch(Condition="Equal (Class)"(A=@n10, B=&PCGPartitionActor))
   > n11 = Branch(Condition=Get Debug())
   > n12 = LogString(InString="To String (Name)"(InName=@n10.Label))
   n13 = FindMeshGrassActorByPartitionBounds!(
    execute=[@n12.then, @n11.else]
    self
    Debug:bool=Get Debug()
    GridExtent:double=Get GridExtent()
    GridMargin:double=Get GridMargin()
    TargetBox:Box Structure=@n10.Bounds
    TargetClass:String=Get TargetClass()
   ) -> (then, TargetActor:Actor Object Reference, Success:bool) {is=title K2Node_CallFunction}
   > n14 = Branch(Condition=@n13.Success)
   > Add(
    TargetArray:Array of Actor Object References=Get TargetNeighborhood() -> (TargetNeighborhood:Array of Actor Object References)
    NewItem:"Actor Object Reference (by ref)"=@n13
   ) {is=title K2Node_CallArrayFunction}
   > Branch(Condition=Get Debug())
   > LogString(InString=GetDisplayName(Object=@n13))
   Set IncompleteNeighborhood(execute=@n14.else, IncompleteNeighborhood:bool=true) -> (Output_Get:bool)
   EvaluateTargetActors!(
    execute=@n2.then_1
    self
    TargetActor:Array of Actor Object References=Get TargetNeighborhood()
    MeshCountMedium:"Integer (by ref)"=Get MeshCountNearbyMedium() -> (MeshCountNearbyMedium:int=0)
    MeshCountEpic:"Integer (by ref)"=Get MeshCountNearbyEpic() -> (MeshCountNearbyEpic:int=0)
    InstanceCountMedium:"Integer (by ref)"=Get InstanceCountNearbyMedium() -> (InstanceCountNearbyMedium:int=0)
    InstanceCountEpic:"Integer (by ref)"=Get InstanceCountNearbyEpic() -> (InstanceCountNearbyEpic:int=0)
   ) -> (then) {is=title K2Node_CallFunction}
   n15 = Sequence!(execute=@n1.then_1) -> (then_0, then_1, then_2:Exec)
   > n16 = Get Editor Camera Info()
   n17 = MakeVector(X=@n16, Y=@n16.CameraLocation_Y)
   > n18 = GetIntersectingActorDescs(InBox_Min=@n17, InBox_Max=@n17)
   > Branch(Condition=@n18.ReturnValue)
   > n19 = For Each Loop with Break!(Exec:Exec, Array:Array of Actor Desc Structures=@n18, Break:Exec=[@n21.else, @n22.then]) -> (LoopBody:Exec, Array Element:Actor Desc Structure, Array Index:int, Completed:Exec)
   n20 = Break Actor Desc(ActorDesc=@n19) -> (bIsSpatiallyLoaded=~, bActorIsEditorOnly=~)
   > Branch(Condition="Equal (Class)"(A=@n20, B=&PCGPartitionActor))
   > Set TargetBox(TargetBox:Box Structure=@n20.Bounds) -> (Output_Get:Box Structure)
   > n21 = Branch(Condition=Get Debug())
   > n22 = LogString(InString="To String (Name)"(InName=@n20.Label))
   n23 = FindMeshGrassActorByPartitionBounds(
    execute=@n15.then_1
    Debug=Get Debug()
    GridExtent=Get GridExtent()
    GridMargin=Get GridMargin()
    TargetBox=Get TargetBox()
    TargetClass=Get TargetClass()
   )
   > Set TargetActor(TargetActor:Actor Object Reference=@n23) -> (Output_Get:Actor Object Reference)
   n24 = Get TargetActor() -> (TargetActor:Actor Object Reference)
   Is Valid(exec=@n15.then_2, InputObject=@n24)
   > EvaluateTargetActors(
    TargetActor=Make Array("[0]":Actor Object Reference=@n24) -> (Array:Array of Actor Object References)
    MeshCountMedium=Get MeshCountMedium() -> (MeshCountMedium:int=0)
    MeshCountEpic=Get MeshCountEpic() -> (MeshCountEpic:int=0)
    InstanceCountMedium=Get InstanceCountMedium() -> (InstanceCountMedium:int=0)
    InstanceCountEpic=Get InstanceCountEpic() -> (InstanceCountEpic:int=0)
   )
   n25 = RightChop!(SourceString:String=GetDisplayName(Object=Get TargetActor()), Count:int=13) -> (ReturnValue:String)
   n26 = Get MeshCountNearbyMedium() -> (MeshCountNearbyMedium=0)
   n27 = Get MeshCountMedium() -> (MeshCountMedium=0)
   n28 = Append!(A=ParseInt!(Int:int=@n27) -> (ReturnValue:String) {is=title K2Node_CallFunction}, B=" | ", C:String="To String (Integer)"(InInt=@n26)) -> (ReturnValue)
   n29 = Get MeshCountNearbyEpic() -> (MeshCountNearbyEpic=0)
   n30 = Get MeshCountEpic() -> (MeshCountEpic=0)
   n31 = Append(A=ParseInt(Int=int + int(A=@n27, B=@n30)), B=" | ", C="To String (Integer)"(InInt=int + int(A=@n26, B=@n29)))
   n32 = Get InstanceCountNearbyMedium() -> (InstanceCountNearbyMedium=0)
   n33 = Get InstanceCountNearbyEpic() -> (InstanceCountNearbyEpic=0)
   n34 = Get InstanceCountMedium() -> (InstanceCountMedium=0)
   n35 = Get InstanceCountEpic() -> (InstanceCountEpic=0)
   n36 = Append(A=ParseInt(Int=int + int(A=@n34, B=@n35)), B=" | ", C="To String (Integer)"(InInt=int + int(A=@n32, B=@n33)))
   n37 = Append(A=ParseInt(Int=@n34), B=" | ", C="To String (Integer)"(InInt=@n32))
   > Set PartitionStats(PartitionStats:String=Append(
     A=Append(A="Partition: ", B=@n25, C="\r\n")
     B=Append!(A=" Components", B="\r\n  Medium : ", C=@n28, D:String="\r\n  Epic   : ", E:String=@n31, F:String="\r\n") -> (ReturnValue)
     C=Append!(A=" Instances", B="\r\n  Medium : ", C=@n37, D="\r\n  Epic   : ", E=@n36) -> (ReturnValue)
    )) -> (Output_Get:String)
   n38 = Sequence!(execute=@n1.then_2) -> (then_0, then_1)
   > n39 = GetSelectedActors()
   > n40 = Filter by Actor Tag!(execute, TargetArray:Array of Actor Object References=@n39, Tag:Name=PCG_MeshGrass, FilterType:EEditorScriptingFilterType Enum) -> (then, ReturnValue:Array of Actor Object References)
   > Branch(Condition=Is Not Empty(TargetArray:Array of Actor Object References=@n40))
   n41 = GetDisplayName(Object="Get (a copy)"(Array:Array of Actor Object References=@n40) -> (Output:Actor Object Reference))
   n42 = RightChop(SourceString=@n41, Count=13)
   n43 = StartsWith(SourceString=@n41, InPrefix=BP_MeshGrass_)
   n44 = Length(TargetArray=@n40)
   > Set SelectionName(SelectionName:String=SelectString(A=Multiple..., B=SelectString(A=@n42, B=@n41, bPickA=@n43), bPickA="integer > integer"(A=@n44, B=1))) -> (Output_Get:String)
   > n45 = For Each Loop(Array:Array of Actor Object References=@n40) -> (Array Element:Actor Object Reference)
   Branch(execute=@n45.Completed, Condition="integer > integer"(A=Get SelectionInstances() -> (SelectionInstances:int=0)))
   n46 = Get SelectionInstances() -> (SelectionInstances=0)
   > Set SelectionStats(SelectionStats:String=Append!(A="Selection: ", B=Get SelectionName() -> (SelectionName:String), C="\r\n Points  : ", D="To String (Integer)"(InInt=@n46)) -> (ReturnValue)) -> (Output_Get:String)
   n47 = GetTypedInputs!(
    execute=@n45.LoopBody
    InCollection:"PCGData Collection Structure (by ref)"=GetGeneratedGraphOutput(self=GetComponentByClass(self=@n45, ComponentClass=&PCGComponent))
    InDataTypeClass:PCGData Class Reference=/Script/PCG.PCGPointData
   ) -> (then, OutTaggedData:Array of PCGTagged Data Structures, ReturnValue:Array of PCGPoint Data Object References)
   > Branch(Condition=Is Not Empty(TargetArray:Array of PCGPoint Data Object References=@n47.ReturnValue))
   n48 = "Get (a copy)"(Array:Array of PCGPoint Data Object References=@n47.ReturnValue) -> (Output:PCGPoint Data Object Reference)
   > Set SelectionInstances(SelectionInstances:int=int + int(A=Get SelectionInstances() -> (SelectionInstances=0), B=GetNumPoints!(self:PCGBase Point Data Object Reference=@n48) -> (ReturnValue:int=0))) -> (Output_Get:int=0)
   n49 = Get SelectionStats() -> (SelectionStats:String)
   n50 = IsEmpty(InString=@n49)
   n51 = Get PartitionStats() -> (PartitionStats:String)
   n52 = Branch(execute=@n38.then_1, Condition=AND Boolean(A=IsEmpty(InString=@n51), B=@n50))
   n53 = Get ETB_Perf() -> (ETB_Perf:Editor Utility Editable Text Box Object Reference)
   > "SetText (Text Box)"!(execute, self:Text Box Object Reference=@n53, InText:Text) -> (then)
   > SetToolTipText!(
    execute
    self:Widget Object Reference=Get BT_Information() -> (BT_Information:Editor Utility Button Object Reference)
    InToolTipText:"Text (by ref)"="To Text (String)"(InString="Current Errors:\r\n• No perf data was found, possibly because the viewport is in an unloaded or irrelevant area")
   ) -> (then)
   > SetColorAndOpacity!(
    execute
    self=Get T_Information() -> (T_Information:Text Object Reference)
    InColorAndOpacity_SpecifiedColor:Linear Color Structure=(R=1,G=0.04,B=0.04,A=1)
    InColorAndOpacity_ColorUseRule:ESlateColorStylingMode Enum
   ) -> (then)
   n54 = SelectString(B="\r\n", bPickA=@n50)
   "SetText (Text Box)"(execute=@n52.else, self=@n53, InText="To Text (String)"(InString=Append!(A=@n51, B=@n54, C=@n49) -> (ReturnValue)))
   n55 = MakeLiteralString(Value="Metrics Infomation:\r\nComponents = Meshes, including duplicates if there is the same mesh with different materials\r\nInstances = Sum of instances across all components, broken down by quality level\r\nPoints = Corresponds to the number of instances this actor will contribute at the highest quality level\r\n\r\nPartition is selected based on viewport location, numbers after separator include adjacent partitions.")
   n56 = Append!(
    A=MakeLiteralString(Value="Current Warnings:\r\n• Some adjacent partitions are not loaded or empty, this could cause metrics to appear misleadingly low.\r\n\r\n")
    B=@n55
   ) -> (ReturnValue)
   n57 = Get IncompleteNeighborhood() -> (IncompleteNeighborhood:bool)
   > SetToolTipText(self=Get BT_Information(), InToolTipText="To Text (String)"(InString=SelectString(A=@n56, B=@n55, bPickA=@n57)))
   > SetColorAndOpacity(
    self=Get T_Information()
    InColorAndOpacity_SpecifiedColor=SelectColor(A=(R=1,G=0.8051,B=0,A=1), B=(R=1,G=1,B=1,A=1), bPickA=Get IncompleteNeighborhood())
   )
  ]}
]
variables []
widgets {
 info {rootWidgetClass=&VerticalBox, widgetCount=53}
 widgets [
  VerticalBox_0 {widgetClassPath=&VerticalBox, widget=VerticalBox_0, properties {clipping=ClipToBounds}}
  Border_86 {
   widgetClassPath=&Border
   parent=VerticalBox_0
   widget=Border_86
   properties {horizontalAlignment=HAlign_Center, brushColor=<rgba 0.0395 0.0395 0.0395 0>}
   slot {class=VerticalBoxSlot}
  }
  TextBlock_34 {
   widgetClassPath=&TextBlock
   parent=Border_86
   widget=TextBlock_34
   properties {
    justification=Center
    text=VISIBILITY TOGGLES
    colorAndOpacity {colorUseRule=UseColor_Specified, specifiedColor=<rgba 0.66 0.66 0.66 1>}
    font {
     bForceMonospaced=false
     bMaterialIsStencil=false
     fontMaterial=null
     fontObject=&Roboto
     letterSpacing=100
     monospacedWidth=1
     size=9
     skewAmount=0
     typefaceFontName=Bold
     outlineSettings {
      bApplyOutlineToDropShadows=false
      bMiteredCorners=false
      bSeparateFillAlpha=false
      outlineMaterial=null
      outlineSize=0
      outlineColor=<rgba 0 0 0 1>
     }
    }
   }
   slot {class=BorderSlot, properties {horizontalAlignment=HAlign_Center}}
  }
  HorizontalBox_1 {
   widgetClassPath=&HorizontalBox
   parent=VerticalBox_0
   widget=HorizontalBox_1
   slot {class=VerticalBoxSlot, properties {padding=<ltrb 2 2 2 2>}}
  }
  EditorUtilityButton_23 {
   widgetClassPath=&EditorUtilityButton
   parent=HorizontalBox_1
   widget=EditorUtilityButton_23
   properties {toolTipText="Affects Mesh, Spline, Texture && Custom generators"}
   slot {class=HorizontalBoxSlot, properties {padding=<ltrb 1 1 1 1>, size {sizeRule=Fill, value=1}}}
  }
  TextBlock_23 {
   widgetClassPath=&TextBlock
   parent=EditorUtilityButton_23
   widget=TextBlock_23
   properties {
    clipping=ClipToBounds
    text=Standalone
    font {
     bForceMonospaced=false
     bMaterialIsStencil=false
     fontMaterial=null
     fontObject=&Roboto
     letterSpacing=0
     monospacedWidth=1
     size=12
     skewAmount=0
     typefaceFontName=Regular
     outlineSettings {
      bApplyOutlineToDropShadows=false
      bMiteredCorners=false
      bSeparateFillAlpha=false
      outlineMaterial=null
      outlineSize=0
      outlineColor=<rgba 0 0 0 1>
     }
    }
   }
   slot {class=ButtonSlot, properties {padding=<ltrb 2 4 2 3>}}
  }
  EditorUtilityButton_4 {
   widgetClassPath=&EditorUtilityButton
   parent=HorizontalBox_1
   widget=EditorUtilityButton_4
   properties {visibility=Collapsed}
   slot {class=HorizontalBoxSlot, properties {padding=<ltrb 1 1 1 1>, size {sizeRule=Fill, value=0.25}}}
  }
  TextBlock_14 {
   widgetClassPath=&TextBlock
   parent=EditorUtilityButton_4
   widget=TextBlock_14
   properties {
    clipping=ClipToBounds
    text=Flip
    font {
     bForceMonospaced=false
     bMaterialIsStencil=false
     fontMaterial=null
     fontObject=&Roboto
     letterSpacing=0
     monospacedWidth=1
     size=12
     skewAmount=0
     typefaceFontName=Regular
     outlineSettings {
      bApplyOutlineToDropShadows=false
      bMiteredCorners=false
      bSeparateFillAlpha=false
      outlineMaterial=null
      outlineSize=0
      outlineColor=<rgba 0 0 0 1>
     }
    }
   }
   slot {class=ButtonSlot, properties {padding=<ltrb 2 4 2 3>}}
  }
  EditorUtilityButton {
   widgetClassPath=&EditorUtilityButton
   parent=HorizontalBox_1
   widget=EditorUtilityButton
   slot {class=HorizontalBoxSlot, properties {padding=<ltrb 1 1 1 1>, size {sizeRule=Fill, value=1}}}
  }
  TextBlock {
   widgetClassPath=&TextBlock
   parent=EditorUtilityButton
   widget=TextBlock
   properties {
    clipping=ClipToBounds
    text=Landscape
    font {
     bForceMonospaced=false
     bMaterialIsStencil=false
     fontMaterial=null
     fontObject=&Roboto
     letterSpacing=0
     monospacedWidth=1
     size=12
     skewAmount=0
     typefaceFontName=Regular
     outlineSettings {
      bApplyOutlineToDropShadows=false
      bMiteredCorners=false
      bSeparateFillAlpha=false
      outlineMaterial=null
      outlineSize=0
      outlineColor=<rgba 0 0 0 1>
     }
    }
   }
   slot {class=ButtonSlot, properties {padding=<ltrb 2 4 2 3>}}
  }
  Spacer_40 {widgetClassPath=&Spacer, parent=VerticalBox_0, widget=Spacer_40, properties {size=<1 10>}, slot {class=VerticalBoxSlot}}
  Border_2 {
   widgetClassPath=&Border
   parent=VerticalBox_0
   widget=Border_2
   properties {horizontalAlignment=HAlign_Center, brushColor=<rgba 0.0395 0.0395 0.0395 0>}
   slot {class=VerticalBoxSlot}
  }
  TextBlock_7 {
   widgetClassPath=&TextBlock
   parent=Border_2
   widget=TextBlock_7
   properties {
    justification=Center
    text=RENDERING QUALITY
    colorAndOpacity {colorUseRule=UseColor_Specified, specifiedColor=<rgba 0.66 0.66 0.66 1>}
    font {
     bForceMonospaced=false
     bMaterialIsStencil=false
     fontMaterial=null
     fontObject=&Roboto
     letterSpacing=100
     monospacedWidth=1
     size=9
     skewAmount=0
     typefaceFontName=Bold
     outlineSettings {
      bApplyOutlineToDropShadows=false
      bMiteredCorners=false
      bSeparateFillAlpha=false
      outlineMaterial=null
      outlineSize=0
      outlineColor=<rgba 0 0 0 1>
     }
    }
   }
   slot {class=BorderSlot, properties {horizontalAlignment=HAlign_Center}}
  }
  HorizontalBox_3 {
   widgetClassPath=&HorizontalBox
   parent=VerticalBox_0
   widget=HorizontalBox_3
   slot {class=VerticalBoxSlot, properties {padding=<ltrb 2 2 2 2>}}
  }
  EditorUtilityButton_1 {
   widgetClassPath=&EditorUtilityButton
   parent=HorizontalBox_3
   widget=EditorUtilityButton_1
   slot {class=HorizontalBoxSlot, properties {padding=<ltrb 1 1 1 1>, size {sizeRule=Fill, value=1}}}
  }
  TextBlock_8 {
   widgetClassPath=&TextBlock
   parent=EditorUtilityButton_1
   widget=TextBlock_8
   properties {
    clipping=ClipToBounds
    text=Low
    font {
     bForceMonospaced=false
     bMaterialIsStencil=false
     fontMaterial=null
     fontObject=&Roboto
     letterSpacing=0
     monospacedWidth=1
     size=12
     skewAmount=0
     typefaceFontName=Regular
     outlineSettings {
      bApplyOutlineToDropShadows=false
      bMiteredCorners=false
      bSeparateFillAlpha=false
      outlineMaterial=null
      outlineSize=0
      outlineColor=<rgba 0 0 0 1>
     }
    }
   }
   slot {class=ButtonSlot, properties {padding=<ltrb 2 4 2 3>}}
  }
  EditorUtilityButton_3 {
   widgetClassPath=&EditorUtilityButton
   parent=HorizontalBox_3
   widget=EditorUtilityButton_3
   slot {class=HorizontalBoxSlot, properties {padding=<ltrb 1 1 1 1>, size {sizeRule=Fill, value=1}}}
  }
  TextBlock_10 {
   widgetClassPath=&TextBlock
   parent=EditorUtilityButton_3
   widget=TextBlock_10
   properties {
    clipping=ClipToBounds
    text=Medium
    font {
     bForceMonospaced=false
     bMaterialIsStencil=false
     fontMaterial=null
     fontObject=&Roboto
     letterSpacing=0
     monospacedWidth=1
     size=12
     skewAmount=0
     typefaceFontName=Regular
     outlineSettings {
      bApplyOutlineToDropShadows=false
      bMiteredCorners=false
      bSeparateFillAlpha=false
      outlineMaterial=null
      outlineSize=0
      outlineColor=<rgba 0 0 0 1>
     }
    }
   }
   slot {class=ButtonSlot, properties {padding=<ltrb 2 4 2 3>}}
  }
  EditorUtilityButton_2 {
   widgetClassPath=&EditorUtilityButton
   parent=HorizontalBox_3
   widget=EditorUtilityButton_2
   slot {class=HorizontalBoxSlot, properties {padding=<ltrb 1 1 1 1>, size {sizeRule=Fill, value=1}}}
  }
  TextBlock_9 {
   widgetClassPath=&TextBlock
   parent=EditorUtilityButton_2
   widget=TextBlock_9
   properties {
    clipping=ClipToBounds
    text=Epic
    font {
     bForceMonospaced=false
     bMaterialIsStencil=false
     fontMaterial=null
     fontObject=&Roboto
     letterSpacing=0
     monospacedWidth=1
     size=12
     skewAmount=0
     typefaceFontName=Regular
     outlineSettings {
      bApplyOutlineToDropShadows=false
      bMiteredCorners=false
      bSeparateFillAlpha=false
      outlineMaterial=null
      outlineSize=0
      outlineColor=<rgba 0 0 0 1>
     }
    }
   }
   slot {class=ButtonSlot, properties {padding=<ltrb 2 4 2 3>}}
  }
  Spacer_1 {widgetClassPath=&Spacer, parent=VerticalBox_0, widget=Spacer_1, properties {size=<1 10>}, slot {class=VerticalBoxSlot}}
  Border {
   widgetClassPath=&Border
   parent=VerticalBox_0
   widget=Border
   properties {horizontalAlignment=HAlign_Center, brushColor=<rgba 0.0395 0.0395 0.0395 0>}
   slot {class=VerticalBoxSlot}
  }
  TextBlock_1 {
   widgetClassPath=&TextBlock
   parent=Border
   widget=TextBlock_1
   properties {
    justification=Center
    text=GENERATION CONTROLS
    colorAndOpacity {colorUseRule=UseColor_Specified, specifiedColor=<rgba 0.66 0.66 0.66 1>}
    font {
     bForceMonospaced=false
     bMaterialIsStencil=false
     fontMaterial=null
     fontObject=&Roboto
     letterSpacing=100
     monospacedWidth=1
     size=9
     skewAmount=0
     typefaceFontName=Bold
     outlineSettings {
      bApplyOutlineToDropShadows=false
      bMiteredCorners=false
      bSeparateFillAlpha=false
      outlineMaterial=null
      outlineSize=0
      outlineColor=<rgba 0 0 0 1>
     }
    }
   }
   slot {class=BorderSlot, properties {horizontalAlignment=HAlign_Center}}
  }
  HorizontalBox_0 {
   widgetClassPath=&HorizontalBox
   parent=VerticalBox_0
   widget=HorizontalBox_0
   slot {class=VerticalBoxSlot, properties {padding=<ltrb 2 2 2 2>}}
  }
  BT_Repopulate {
   widgetClassPath=&EditorUtilityButton
   parent=HorizontalBox_0
   widget=BT_Repopulate
   slot {class=HorizontalBoxSlot, properties {padding=<ltrb 1 1 1 1>, size {sizeRule=Fill, value=1}}}
  }
  TextBlock_2 {
   widgetClassPath=&TextBlock
   parent=BT_Repopulate
   widget=TextBlock_2
   properties {
    clipping=ClipToBounds
    text=Repopulate
    font {
     bForceMonospaced=false
     bMaterialIsStencil=false
     fontMaterial=null
     fontObject=&Roboto
     letterSpacing=0
     monospacedWidth=1
     size=12
     skewAmount=0
     typefaceFontName=Regular
     outlineSettings {
      bApplyOutlineToDropShadows=false
      bMiteredCorners=false
      bSeparateFillAlpha=false
      outlineMaterial=null
      outlineSize=0
      outlineColor=<rgba 0 0 0 1>
     }
    }
   }
   slot {class=ButtonSlot, properties {padding=<ltrb 0 1 0 0>}}
  }
  BT_Regenerate {
   widgetClassPath=&EditorUtilityButton
   parent=HorizontalBox_0
   widget=BT_Regenerate
   properties {toolTipText="Warning: This will take a long time if a large area is loaded!"}
   slot {class=HorizontalBoxSlot, properties {padding=<ltrb 1 1 1 1>, size {sizeRule=Fill, value=1}}}
  }
  HorizontalBox_2 {
   widgetClassPath=&HorizontalBox
   parent=BT_Regenerate
   widget=HorizontalBox_2
   properties {clipping=ClipToBounds}
   slot {class=ButtonSlot, properties {padding=<ltrb 2 4 2 3>}}
  }
  TextBlock_3 {
   widgetClassPath=&TextBlock
   parent=HorizontalBox_2
   widget=TextBlock_3
   properties {
    clipping=ClipToBounds
    text=Regenerate
    font {
     bForceMonospaced=false
     bMaterialIsStencil=false
     fontMaterial=null
     fontObject=&Roboto
     letterSpacing=0
     monospacedWidth=1
     size=12
     skewAmount=0
     typefaceFontName=Regular
     outlineSettings {
      bApplyOutlineToDropShadows=false
      bMiteredCorners=false
      bSeparateFillAlpha=false
      outlineMaterial=null
      outlineSize=0
      outlineColor=<rgba 0 0 0 1>
     }
    }
   }
   slot {class=HorizontalBoxSlot}
  }
  SizeBox_1 {
   widgetClassPath=&SizeBox
   parent=HorizontalBox_2
   widget=SizeBox_1
   properties {bOverride_HeightOverride=true}
   slot {class=HorizontalBoxSlot}
  }
  TextBlock_6 {
   widgetClassPath=&TextBlock
   parent=SizeBox_1
   widget=TextBlock_6
   properties {
    clipping=ClipToBounds
    text="!"
    font {
     bForceMonospaced=false
     bMaterialIsStencil=false
     fontMaterial=null
     fontObject=&BurbankSmall
     letterSpacing=0
     monospacedWidth=1
     size=15
     skewAmount=0
     typefaceFontName=Bold
     outlineSettings {
      bApplyOutlineToDropShadows=false
      bMiteredCorners=false
      bSeparateFillAlpha=false
      outlineMaterial=null
      outlineSize=0
      outlineColor=<rgba 0 0 0 1>
     }
    }
   }
   slot {class=SizeBoxSlot, properties {padding=<ltrb 1 -3 0 0>}}
  }
  SizeBox_3 {
   widgetClassPath=&SizeBox
   parent=VerticalBox_0
   widget=SizeBox_3
   properties {bOverride_HeightOverride=true, heightOverride=31.6}
   slot {class=VerticalBoxSlot, properties {padding=<ltrb 2 2 2 2>}}
  }
  HorizontalBox_4 {widgetClassPath=&HorizontalBox, parent=SizeBox_3, widget=HorizontalBox_4, slot {class=SizeBoxSlot}}
  BT_Speed_1x {
   widgetClassPath=&EditorUtilityButton
   parent=HorizontalBox_4
   widget=BT_Speed_1x
   properties {toolTipText=Normal Speed}
   slot {class=HorizontalBoxSlot, properties {padding=<ltrb 1 1 1 1>, size {sizeRule=Fill, value=1}}}
  }
  TextBlock_11 {
   widgetClassPath=&TextBlock
   parent=BT_Speed_1x
   widget=TextBlock_11
   properties {
    text="►"
    font {
     bForceMonospaced=false
     bMaterialIsStencil=false
     fontMaterial=null
     fontObject=&Roboto
     letterSpacing=0
     monospacedWidth=1
     size=18
     skewAmount=0
     typefaceFontName=Regular
     outlineSettings {
      bApplyOutlineToDropShadows=false
      bMiteredCorners=false
      bSeparateFillAlpha=false
      outlineMaterial=null
      outlineSize=0
      outlineColor=<rgba 0 0 0 1>
     }
    }
   }
   slot {class=ButtonSlot, properties {padding=<ltrb 0 0 0 0>}}
  }
  BT_Speed_3x {
   widgetClassPath=&EditorUtilityButton
   parent=HorizontalBox_4
   widget=BT_Speed_3x
   properties {toolTipText=3x Speed}
   slot {class=HorizontalBoxSlot, properties {padding=<ltrb 1 1 1 1>, size {sizeRule=Fill, value=1}}}
  }
  TextBlock_12 {
   widgetClassPath=&TextBlock
   parent=BT_Speed_3x
   widget=TextBlock_12
   properties {
    text=">>"
    font {
     bForceMonospaced=false
     bMaterialIsStencil=false
     fontMaterial=null
     fontObject=&NotoSans
     letterSpacing=0
     monospacedWidth=1
     size=15
     skewAmount=0
     typefaceFontName=NotoSans-Bold
     outlineSettings {
      bApplyOutlineToDropShadows=false
      bMiteredCorners=false
      bSeparateFillAlpha=false
      outlineMaterial=null
      outlineSize=0
      outlineColor=<rgba 0 0 0 1>
     }
    }
   }
   slot {class=ButtonSlot, properties {padding=<ltrb 0 -0.5 0 0>}}
  }
  BT_Speed_6x {
   widgetClassPath=&EditorUtilityButton
   parent=HorizontalBox_4
   widget=BT_Speed_6x
   properties {toolTipText=6x Speed}
   slot {class=HorizontalBoxSlot, properties {padding=<ltrb 1 1 1 1>, size {sizeRule=Fill, value=1}}}
  }
  TextBlock_13 {
   widgetClassPath=&TextBlock
   parent=BT_Speed_6x
   widget=TextBlock_13
   properties {
    text=">>>"
    font {
     bForceMonospaced=false
     bMaterialIsStencil=false
     fontMaterial=null
     fontObject=&NotoSans
     letterSpacing=0
     monospacedWidth=1
     size=15
     skewAmount=0
     typefaceFontName=NotoSans-Bold
     outlineSettings {
      bApplyOutlineToDropShadows=false
      bMiteredCorners=false
      bSeparateFillAlpha=false
      outlineMaterial=null
      outlineSize=0
      outlineColor=<rgba 0 0 0 1>
     }
    }
   }
   slot {class=ButtonSlot, properties {padding=<ltrb 0 -0.5 0 0>}}
  }
  Spacer {widgetClassPath=&Spacer, parent=VerticalBox_0, widget=Spacer, properties {size=<1 10>}, slot {class=VerticalBoxSlot}}
  Border_1 {
   widgetClassPath=&Border
   parent=VerticalBox_0
   widget=Border_1
   properties {horizontalAlignment=HAlign_Center, brushColor=<rgba 0.0395 0.0395 0.0395 0>}
   slot {class=VerticalBoxSlot}
  }
  TextBlock_4 {
   widgetClassPath=&TextBlock
   parent=Border_1
   widget=TextBlock_4
   properties {
    justification=Center
    text=PERFORMANCE METRICS
    colorAndOpacity {colorUseRule=UseColor_Specified, specifiedColor=<rgba 0.66 0.66 0.66 1>}
    font {
     bForceMonospaced=false
     bMaterialIsStencil=false
     fontMaterial=null
     fontObject=&Roboto
     letterSpacing=100
     monospacedWidth=1
     size=9
     skewAmount=0
     typefaceFontName=Bold
     outlineSettings {
      bApplyOutlineToDropShadows=false
      bMiteredCorners=false
      bSeparateFillAlpha=false
      outlineMaterial=null
      outlineSize=0
      outlineColor=<rgba 0 0 0 1>
     }
    }
   }
   slot {class=BorderSlot, properties {horizontalAlignment=HAlign_Center}}
  }
  HorizontalBox {
   widgetClassPath=&HorizontalBox
   parent=VerticalBox_0
   widget=HorizontalBox
   slot {class=VerticalBoxSlot, properties {padding=<ltrb 2 2 2 2>}}
  }
  Border_77 {
   widgetClassPath=&Border
   parent=HorizontalBox
   widget=Border_77
   properties {
    background {
     bIsDynamicallyLoaded=false
     drawAs=RoundedBox
     imageType=NoImage
     mirroring=NoMirror
     resourceName=None
     resourceObject=null
     tiling=NoTile
     imageSize=<32 32>
     margin=<ltrb 0.2 0.2 0.2 0.2>
     outlineSettings {
      bUseBrushTransparency=false
      roundingType=FixedRadius
      width=0
      color {colorUseRule=UseColor_Specified, specifiedColor=<rgba 0 0 0 0>}
      cornerRadii=<5 5 5 5>
     }
     tintColor {colorUseRule=UseColor_Specified, specifiedColor=<rgba 1 1 1 1>}
     uVRegion {bIsValid=false, max=<0 0>, min=<0 0>}
    }
    brushColor=<rgba 0.0103 0.0103 0.0103 1>
    padding=<ltrb 0 0 0 0>
   }
   slot {class=HorizontalBoxSlot, properties {verticalAlignment=VAlign_Top, size {sizeRule=Fill, value=1}}}
  }
  Overlay_1 {widgetClassPath=&Overlay, parent=Border_77, widget=Overlay_1, slot {class=BorderSlot, properties {padding=<ltrb 0 0 0 0>}}}
  ETB_Perf {
   widgetClassPath=&EditorUtilityEditableTextBox
   bIsVariable=true
   parent=Overlay_1
   widget=ETB_Perf
   properties {
    isReadOnly=true
    widgetStyle {
     backgroundColor {colorUseRule=UseColor_Specified, specifiedColor=<rgba 1 1 1 1>}
     backgroundImageFocused {
      bIsDynamicallyLoaded=false
      drawAs=RoundedBox
      imageType=NoImage
      mirroring=NoMirror
      resourceName=None
      resourceObject=null
      tiling=NoTile
      imageSize=<0 0>
      margin=<ltrb 0 0 0 0>
      outlineSettings {
       bUseBrushTransparency=false
       roundingType=FixedRadius
       width=1
       color {colorUseRule=UseColor_ColorTable, specifiedColor=<rgba 0 0 0 0>}
       cornerRadii=<4 4 4 4>
      }
      tintColor {colorUseRule=UseColor_Specified, specifiedColor=<rgba 0.0048 0.0048 0.0048 1>}
      uVRegion {bIsValid=false, max=<0 0>, min=<0 0>}
     }
     backgroundImageHovered {
      bIsDynamicallyLoaded=false
      drawAs=RoundedBox
      imageType=NoImage
      mirroring=NoMirror
      resourceName=None
      resourceObject=null
      tiling=NoTile
      imageSize=<0 0>
      margin=<ltrb 0 0 0 0>
      outlineSettings {
       bUseBrushTransparency=false
       roundingType=FixedRadius
       width=1
       color {colorUseRule=UseColor_ColorTable, specifiedColor=<rgba 0 0 0 0>}
       cornerRadii=<4 4 4 4>
      }
      tintColor {colorUseRule=UseColor_Specified, specifiedColor=<rgba 0.0048 0.0048 0.0048 1>}
      uVRegion {bIsValid=false, max=<0 0>, min=<0 0>}
     }
     backgroundImageNormal {
      bIsDynamicallyLoaded=false
      drawAs=RoundedBox
      imageType=NoImage
      mirroring=NoMirror
      resourceName=None
      resourceObject=null
      tiling=NoTile
      imageSize=<0 0>
      margin=<ltrb 0 0 0 0>
      outlineSettings {
       bUseBrushTransparency=false
       roundingType=FixedRadius
       width=1
       color {colorUseRule=UseColor_ColorTable, specifiedColor=<rgba 0 0 0 0>}
       cornerRadii=<4 4 4 4>
      }
      tintColor {colorUseRule=UseColor_Specified, specifiedColor=<rgba 0.0048 0.0048 0.0048 1>}
      uVRegion {bIsValid=false, max=<0 0>, min=<0 0>}
     }
     backgroundImageReadOnly {
      bIsDynamicallyLoaded=false
      drawAs=RoundedBox
      imageType=NoImage
      mirroring=NoMirror
      resourceName=None
      resourceObject=null
      tiling=NoTile
      imageSize=<0 0>
      margin=<ltrb 0 0 0 0>
      outlineSettings {
       bUseBrushTransparency=false
       roundingType=FixedRadius
       width=1
       color {colorUseRule=UseColor_ColorTable, specifiedColor=<rgba 0 0 0 0>}
       cornerRadii=<4 4 4 4>
      }
      tintColor {colorUseRule=UseColor_Specified, specifiedColor=<rgba 0.0284 0.0284 0.0284 1>}
      uVRegion {bIsValid=false, max=<0 0>, min=<0 0>}
     }
     focusedForegroundColor {colorUseRule=UseColor_Specified, specifiedColor=<rgba 1 1 1 1>}
     foregroundColor {colorUseRule=UseColor_Specified, specifiedColor=<rgba 0.5271 0.5271 0.5271 1>}
     hScrollBarPadding=<ltrb 0 0 0 0>
     padding=<ltrb 8 8 4 8>
     readOnlyForegroundColor {colorUseRule=UseColor_Foreground, specifiedColor=<rgba 1 0 1 1>}
     scrollBarStyle {
      thickness=8
      draggedThumbImage {
       bIsDynamicallyLoaded=false
       drawAs=RoundedBox
       imageType=NoImage
       mirroring=NoMirror
       resourceName=None
       resourceObject=null
       tiling=NoTile
       imageSize=<0 0>
       margin=<ltrb 0 0 0 0>
       outlineSettings {
        bUseBrushTransparency=false
        roundingType=FixedRadius
        width=0
        color {colorUseRule=UseColor_Specified, specifiedColor=<rgba 0 0 0 0>}
        cornerRadii=<4 4 4 4>
       }
       tintColor {colorUseRule=UseColor_Specified, specifiedColor=<rgba 0.2159 0.2159 0.2159 1>}
       uVRegion {bIsValid=false, max=<0 0>, min=<0 0>}
      }
      horizontalBackgroundImage {
       bIsDynamicallyLoaded=false
       drawAs=NoDrawType
       imageType=NoImage
       mirroring=NoMirror
       resourceName=None
       resourceObject=null
       tiling=NoTile
       imageSize=<0 0>
       margin=<ltrb 0 0 0 0>
       outlineSettings {
        bUseBrushTransparency=false
        roundingType=HalfHeightRadius
        width=0
        color {colorUseRule=UseColor_Specified, specifiedColor=<rgba 0 0 0 0>}
        cornerRadii=<0 0 0 0>
       }
       tintColor {colorUseRule=UseColor_Specified, specifiedColor=<rgba 1 1 1 1>}
       uVRegion {bIsValid=false, max=<0 0>, min=<0 0>}
      }
      horizontalBottomSlotImage {
       bIsDynamicallyLoaded=false
       drawAs=NoDrawType
       imageType=NoImage
       mirroring=NoMirror
       resourceName=None
       resourceObject=null
       tiling=NoTile
       imageSize=<0 0>
       margin=<ltrb 0 0 0 0>
       outlineSettings {
        bUseBrushTransparency=false
        roundingType=HalfHeightRadius
        width=0
        color {colorUseRule=UseColor_Specified, specifiedColor=<rgba 0 0 0 0>}
        cornerRadii=<0 0 0 0>
       }
       tintColor {colorUseRule=UseColor_Specified, specifiedColor=<rgba 1 1 1 1>}
       uVRegion {bIsValid=false, max=<0 0>, min=<0 0>}
      }
      horizontalTopSlotImage {
       bIsDynamicallyLoaded=false
       drawAs=NoDrawType
       imageType=NoImage
       mirroring=NoMirror
       resourceName=None
       resourceObject=null
       tiling=NoTile
       imageSize=<0 0>
       margin=<ltrb 0 0 0 0>
       outlineSettings {
        bUseBrushTransparency=false
        roundingType=HalfHeightRadius
        width=0
        color {colorUseRule=UseColor_Specified, specifiedColor=<rgba 0 0 0 0>}
        cornerRadii=<0 0 0 0>
       }
       tintColor {colorUseRule=UseColor_Specified, specifiedColor=<rgba 1 1 1 1>}
       uVRegion {bIsValid=false, max=<0 0>, min=<0 0>}
      }
      hoveredThumbImage {
       bIsDynamicallyLoaded=false
       drawAs=RoundedBox
       imageType=NoImage
       mirroring=NoMirror
       resourceName=None
       resourceObject=null
       tiling=NoTile
       imageSize=<0 0>
       margin=<ltrb 0 0 0 0>
       outlineSettings {
        bUseBrushTransparency=false
        roundingType=FixedRadius
        width=0
        color {colorUseRule=UseColor_Specified, specifiedColor=<rgba 0 0 0 0>}
        cornerRadii=<4 4 4 4>
       }
       tintColor {colorUseRule=UseColor_Specified, specifiedColor=<rgba 0.2159 0.2159 0.2159 1>}
       uVRegion {bIsValid=false, max=<0 0>, min=<0 0>}
      }
      normalThumbImage {
       bIsDynamicallyLoaded=false
       drawAs=RoundedBox
       imageType=NoImage
       mirroring=NoMirror
       resourceName=None
       resourceObject=null
       tiling=NoTile
       imageSize=<0 0>
       margin=<ltrb 0 0 0 0>
       outlineSettings {
        bUseBrushTransparency=false
        roundingType=FixedRadius
        width=0
        color {colorUseRule=UseColor_Specified, specifiedColor=<rgba 0 0 0 0>}
        cornerRadii=<4 4 4 4>
       }
       tintColor {colorUseRule=UseColor_Specified, specifiedColor=<rgba 0.0953 0.0953 0.0953 1>}
       uVRegion {bIsValid=false, max=<0 0>, min=<0 0>}
      }
      verticalBackgroundImage {
       bIsDynamicallyLoaded=false
       drawAs=NoDrawType
       imageType=NoImage
       mirroring=NoMirror
       resourceName=None
       resourceObject=null
       tiling=NoTile
       imageSize=<0 0>
       margin=<ltrb 0 0 0 0>
       outlineSettings {
        bUseBrushTransparency=false
        roundingType=HalfHeightRadius
        width=0
        color {colorUseRule=UseColor_Specified, specifiedColor=<rgba 0 0 0 0>}
        cornerRadii=<0 0 0 0>
       }
       tintColor {colorUseRule=UseColor_Specified, specifiedColor=<rgba 1 1 1 1>}
       uVRegion {bIsValid=false, max=<0 0>, min=<0 0>}
      }
      verticalBottomSlotImage {
       bIsDynamicallyLoaded=false
       drawAs=NoDrawType
       imageType=NoImage
       mirroring=NoMirror
       resourceName=None
       resourceObject=null
       tiling=NoTile
       imageSize=<0 0>
       margin=<ltrb 0 0 0 0>
       outlineSettings {
        bUseBrushTransparency=false
        roundingType=HalfHeightRadius
        width=0
        color {colorUseRule=UseColor_Specified, specifiedColor=<rgba 0 0 0 0>}
        cornerRadii=<0 0 0 0>
       }
       tintColor {colorUseRule=UseColor_Specified, specifiedColor=<rgba 1 1 1 1>}
       uVRegion {bIsValid=false, max=<0 0>, min=<0 0>}
      }
      verticalTopSlotImage {
       bIsDynamicallyLoaded=false
       drawAs=NoDrawType
       imageType=NoImage
       mirroring=NoMirror
       resourceName=None
       resourceObject=null
       tiling=NoTile
       imageSize=<0 0>
       margin=<ltrb 0 0 0 0>
       outlineSettings {
        bUseBrushTransparency=false
        roundingType=HalfHeightRadius
        width=0
        color {colorUseRule=UseColor_Specified, specifiedColor=<rgba 0 0 0 0>}
        cornerRadii=<0 0 0 0>
       }
       tintColor {colorUseRule=UseColor_Specified, specifiedColor=<rgba 1 1 1 1>}
       uVRegion {bIsValid=false, max=<0 0>, min=<0 0>}
      }
     }
     textStyle {
      overflowPolicy=Clip
      transformPolicy=None
      colorAndOpacity {colorUseRule=UseColor_Foreground, specifiedColor=<rgba 1 0 1 1>}
      font {
       bForceMonospaced=true
       bMaterialIsStencil=false
       fontMaterial=null
       fontObject=/Engine/EngineFonts/DroidSansMono
       letterSpacing=0
       monospacedWidth=1
       size=9
       skewAmount=0
       typefaceFontName=Default
       outlineSettings {
        bApplyOutlineToDropShadows=false
        bMiteredCorners=false
        bSeparateFillAlpha=false
        outlineMaterial=null
        outlineSize=0
        outlineColor=<rgba 0 0 0 1>
       }
      }
      highlightColor {colorUseRule=UseColor_ColorTable, specifiedColor=<rgba 0 0 0 0>}
      highlightShape {
       bIsDynamicallyLoaded=false
       drawAs=Image
       imageType=NoImage
       mirroring=NoMirror
       resourceName=None
       resourceObject=null
       tiling=NoTile
       imageSize=<0 0>
       margin=<ltrb 0 0 0 0>
       outlineSettings {
        bUseBrushTransparency=false
        roundingType=HalfHeightRadius
        width=0
        color {colorUseRule=UseColor_Specified, specifiedColor=<rgba 0 0 0 0>}
        cornerRadii=<0 0 0 0>
       }
       tintColor {colorUseRule=UseColor_ColorTable, specifiedColor=<rgba 0 0 0 0>}
       uVRegion {bIsValid=false, max=<0 0>, min=<0 0>}
      }
      selectedBackgroundColor {colorUseRule=UseColor_ColorTable, specifiedColor=<rgba 0 0 0 0>}
      shadowColorAndOpacity=<rgba 0 0 0 1>
      shadowOffset=<0 0>
      strikeBrush {
       bIsDynamicallyLoaded=false
       drawAs=Image
       imageType=NoImage
       mirroring=NoMirror
       resourceName=None
       resourceObject=null
       tiling=NoTile
       imageSize=<32 32>
       margin=<ltrb 0 0 0 0>
       outlineSettings {
        bUseBrushTransparency=false
        roundingType=HalfHeightRadius
        width=0
        color {colorUseRule=UseColor_Specified, specifiedColor=<rgba 0 0 0 0>}
        cornerRadii=<0 0 0 0>
       }
       tintColor {colorUseRule=UseColor_Specified, specifiedColor=<rgba 1 1 1 1>}
       uVRegion {bIsValid=false, max=<0 0>, min=<0 0>}
      }
      underlineBrush {
       bIsDynamicallyLoaded=false
       drawAs=Image
       imageType=NoImage
       mirroring=NoMirror
       resourceName=None
       resourceObject=null
       tiling=NoTile
       imageSize=<32 32>
       margin=<ltrb 0 0 0 0>
       outlineSettings {
        bUseBrushTransparency=false
        roundingType=HalfHeightRadius
        width=0
        color {colorUseRule=UseColor_Specified, specifiedColor=<rgba 0 0 0 0>}
        cornerRadii=<0 0 0 0>
       }
       tintColor {colorUseRule=UseColor_Specified, specifiedColor=<rgba 1 1 1 1>}
       uVRegion {bIsValid=false, max=<0 0>, min=<0 0>}
      }
     }
     vScrollBarPadding=<ltrb 0 0 0 0>
    }
   }
   slot {class=OverlaySlot, properties {horizontalAlignment=HAlign_Fill, verticalAlignment=VAlign_Fill}}
  }
  BasicGradientFill {
   widgetClassPath=/Game/UI/Foundation/Stats/BasicGradientFill
   bIsVariable=true
   parent=Overlay_1
   widget=BasicGradientFill
   properties {rotation=0.25, brush Size=<20 1>, color 1=<rgba 0.0284 0.0284 0.0284 0>, color 2=<rgba 0.0284 0.0284 0.0284 1>}
   slot {class=OverlaySlot, properties {horizontalAlignment=HAlign_Right, verticalAlignment=VAlign_Fill, padding=<ltrb 0 2 8 2>}}
  }
  BT_Repopulate_1 {
   widgetClassPath=&EditorUtilityButton
   parent=Overlay_1
   widget=BT_Repopulate_1
   properties {toolTipText=Refresh Metrics}
   slot {class=OverlaySlot, properties {horizontalAlignment=HAlign_Right, padding=<ltrb 0 3 3 0>}}
  }
  TextBlock_5 {
   widgetClassPath=&TextBlock
   parent=BT_Repopulate_1
   widget=TextBlock_5
   properties {
    text="↻"
    font {
     bForceMonospaced=false
     bMaterialIsStencil=false
     fontMaterial=null
     fontObject=&Roboto
     letterSpacing=0
     monospacedWidth=1
     size=12
     skewAmount=0
     typefaceFontName=Bold
     outlineSettings {
      bApplyOutlineToDropShadows=false
      bMiteredCorners=false
      bSeparateFillAlpha=false
      outlineMaterial=null
      outlineSize=0
      outlineColor=<rgba 0 0 0 1>
     }
    }
   }
   slot {class=ButtonSlot, properties {padding=<ltrb -6 1 -6 0>}}
  }
  BT_Information {
   widgetClassPath=&EditorUtilityButton
   parent=Overlay_1
   widget=BT_Information
   properties {
    isFocusable=false
    toolTipText="Press ↻ button to get started!"
    widgetStyle {
     clickedSlateSound {resourceObject=null}
     disabled {
      bIsDynamicallyLoaded=false
      drawAs=RoundedBox
      imageType=NoImage
      mirroring=NoMirror
      resourceName=None
      resourceObject=null
      tiling=NoTile
      imageSize=<0 0>
      margin=<ltrb 0 0 0 0>
      outlineSettings {
       bUseBrushTransparency=false
       roundingType=FixedRadius
       width=1
       color {colorUseRule=UseColor_ColorTable, specifiedColor=<rgba 0 0 0 0>}
       cornerRadii=<4 4 4 4>
      }
      tintColor {colorUseRule=UseColor_Specified, specifiedColor=<rgba 0.0395 0.0395 0.0395 1>}
      uVRegion {bIsValid=false, max=<0 0>, min=<0 0>}
     }
     disabledForeground {colorUseRule=UseColor_ColorTable, specifiedColor=<rgba 0 0 0 0>}
     hovered {
      bIsDynamicallyLoaded=false
      drawAs=RoundedBox
      imageType=NoImage
      mirroring=NoMirror
      resourceName=None
      resourceObject=null
      tiling=NoTile
      imageSize=<32 32>
      margin=<ltrb 0 0 0 0>
      outlineSettings {
       bUseBrushTransparency=false
       roundingType=FixedRadius
       width=1
       color {colorUseRule=UseColor_ColorTable, specifiedColor=<rgba 0 0 0 0>}
       cornerRadii=<4 4 4 4>
      }
      tintColor {colorUseRule=UseColor_Specified, specifiedColor=<rgba 0.0953 0.0953 0.0953 1>}
      uVRegion {bIsValid=false, max=<0 0>, min=<0 0>}
     }
     hoveredForeground {colorUseRule=UseColor_ColorTable, specifiedColor=<rgba 0 0 0 0>}
     hoveredSlateSound {resourceObject=null}
     normal {
      bIsDynamicallyLoaded=false
      drawAs=RoundedBox
      imageType=NoImage
      mirroring=NoMirror
      resourceName=None
      resourceObject=null
      tiling=NoTile
      imageSize=<32 32>
      margin=<ltrb 0 0 0 0>
      outlineSettings {
       bUseBrushTransparency=false
       roundingType=FixedRadius
       width=1
       color {colorUseRule=UseColor_ColorTable, specifiedColor=<rgba 0 0 0 0>}
       cornerRadii=<4 4 4 4>
      }
      tintColor {colorUseRule=UseColor_Specified, specifiedColor=<rgba 0.0395 0.0395 0.0395 1>}
      uVRegion {bIsValid=false, max=<0 0>, min=<0 0>}
     }
     normalForeground {colorUseRule=UseColor_ColorTable, specifiedColor=<rgba 0 0 0 0>}
     normalPadding=<ltrb 12 1.5 12 1.5>
     pressed {
      bIsDynamicallyLoaded=false
      drawAs=RoundedBox
      imageType=NoImage
      mirroring=NoMirror
      resourceName=None
      resourceObject=null
      tiling=NoTile
      imageSize=<32 32>
      margin=<ltrb 0 0 0 0>
      outlineSettings {
       bUseBrushTransparency=false
       roundingType=FixedRadius
       width=1
       color {colorUseRule=UseColor_ColorTable, specifiedColor=<rgba 0 0 0 0>}
       cornerRadii=<4 4 4 4>
      }
      tintColor {colorUseRule=UseColor_Specified, specifiedColor=<rgba 0.0395 0.0395 0.0395 1>}
      uVRegion {bIsValid=false, max=<0 0>, min=<0 0>}
     }
     pressedForeground {colorUseRule=UseColor_ColorTable, specifiedColor=<rgba 0 0 0 0>}
     pressedPadding=<ltrb 12 1.5 12 1.5>
     pressedSlateSound {resourceObject=null}
    }
   }
   slot {class=OverlaySlot, properties {horizontalAlignment=HAlign_Right, padding=<ltrb 0 3 34 0>}}
  }
  T_Information {
   widgetClassPath=&TextBlock
   bIsVariable=true
   parent=BT_Information
   widget=T_Information
   properties {
    text="ⓘ"
    font {
     bForceMonospaced=false
     bMaterialIsStencil=false
     fontMaterial=null
     fontObject=&Roboto
     letterSpacing=0
     monospacedWidth=1
     size=12
     skewAmount=0
     typefaceFontName=Bold
     outlineSettings {
      bApplyOutlineToDropShadows=false
      bMiteredCorners=false
      bSeparateFillAlpha=false
      outlineMaterial=null
      outlineSize=0
      outlineColor=<rgba 0 0 0 1>
     }
    }
   }
   slot {class=ButtonSlot, properties {padding=<ltrb -6 1 -6 0>}}
  }
  BT_Heatmap {
   widgetClassPath=&EditorUtilityButton
   parent=Overlay_1
   widget=BT_Heatmap
   properties {
    isFocusable=false
    toolTipText="Press ↻ button to get started!"
    widgetStyle {
     clickedSlateSound {resourceObject=null}
     disabled {
      bIsDynamicallyLoaded=false
      drawAs=RoundedBox
      imageType=NoImage
      mirroring=NoMirror
      resourceName=None
      resourceObject=null
      tiling=NoTile
      imageSize=<0 0>
      margin=<ltrb 0 0 0 0>
      outlineSettings {
       bUseBrushTransparency=false
       roundingType=FixedRadius
       width=1
       color {colorUseRule=UseColor_ColorTable, specifiedColor=<rgba 0 0 0 0>}
       cornerRadii=<4 4 4 4>
      }
      tintColor {colorUseRule=UseColor_Specified, specifiedColor=<rgba 0.0395 0.0395 0.0395 1>}
      uVRegion {bIsValid=false, max=<0 0>, min=<0 0>}
     }
     disabledForeground {colorUseRule=UseColor_ColorTable, specifiedColor=<rgba 0 0 0 0>}
     hovered {
      bIsDynamicallyLoaded=false
      drawAs=RoundedBox
      imageType=NoImage
      mirroring=NoMirror
      resourceName=None
      resourceObject=null
      tiling=NoTile
      imageSize=<32 32>
      margin=<ltrb 0 0 0 0>
      outlineSettings {
       bUseBrushTransparency=false
       roundingType=FixedRadius
       width=1
       color {colorUseRule=UseColor_ColorTable, specifiedColor=<rgba 0 0 0 0>}
       cornerRadii=<4 4 4 4>
      }
      tintColor {colorUseRule=UseColor_Specified, specifiedColor=<rgba 0.0953 0.0953 0.0953 1>}
      uVRegion {bIsValid=false, max=<0 0>, min=<0 0>}
     }
     hoveredForeground {colorUseRule=UseColor_ColorTable, specifiedColor=<rgba 0 0 0 0>}
     hoveredSlateSound {resourceObject=null}
     normal {
      bIsDynamicallyLoaded=false
      drawAs=RoundedBox
      imageType=NoImage
      mirroring=NoMirror
      resourceName=None
      resourceObject=null
      tiling=NoTile
      imageSize=<32 32>
      margin=<ltrb 0 0 0 0>
      outlineSettings {
       bUseBrushTransparency=false
       roundingType=FixedRadius
       width=1
       color {colorUseRule=UseColor_ColorTable, specifiedColor=<rgba 0 0 0 0>}
       cornerRadii=<4 4 4 4>
      }
      tintColor {colorUseRule=UseColor_Specified, specifiedColor=<rgba 0.0395 0.0395 0.0395 1>}
      uVRegion {bIsValid=false, max=<0 0>, min=<0 0>}
     }
     normalForeground {colorUseRule=UseColor_ColorTable, specifiedColor=<rgba 0 0 0 0>}
     normalPadding=<ltrb 12 1.5 12 1.5>
     pressed {
      bIsDynamicallyLoaded=false
      drawAs=RoundedBox
      imageType=NoImage
      mirroring=NoMirror
      resourceName=None
      resourceObject=null
      tiling=NoTile
      imageSize=<32 32>
      margin=<ltrb 0 0 0 0>
      outlineSettings {
       bUseBrushTransparency=false
       roundingType=FixedRadius
       width=1
       color {colorUseRule=UseColor_ColorTable, specifiedColor=<rgba 0 0 0 0>}
       cornerRadii=<4 4 4 4>
      }
      tintColor {colorUseRule=UseColor_Specified, specifiedColor=<rgba 0.0395 0.0395 0.0395 1>}
      uVRegion {bIsValid=false, max=<0 0>, min=<0 0>}
     }
     pressedForeground {colorUseRule=UseColor_ColorTable, specifiedColor=<rgba 0 0 0 0>}
     pressedPadding=<ltrb 12 1.5 12 1.5>
     pressedSlateSound {resourceObject=null}
    }
   }
   slot {class=OverlaySlot, properties {horizontalAlignment=HAlign_Right, padding=<ltrb 0 3 66 0>}}
  }
  T_Information_1 {
   widgetClassPath=&TextBlock
   bIsVariable=true
   parent=BT_Heatmap
   widget=T_Information_1
   properties {
    text="▦"
    font {
     bForceMonospaced=false
     bMaterialIsStencil=false
     fontMaterial=null
     fontObject=&BurbankSmall
     letterSpacing=0
     monospacedWidth=1
     size=12
     skewAmount=0
     typefaceFontName=Bold
     outlineSettings {
      bApplyOutlineToDropShadows=false
      bMiteredCorners=false
      bSeparateFillAlpha=false
      outlineMaterial=null
      outlineSize=0
      outlineColor=<rgba 0 0 0 1>
     }
    }
   }
   slot {class=ButtonSlot, properties {padding=<ltrb -6 1 -6 0>}}
  }
 ]
}
